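; Looping gameplay cue notify that drives tactical sprint energy drain audio for the locally viewed player.
; On loop start, it caches the target, audio components and TacticalSprintEnergyComponent, then starts a repeating timer calling UpdateFX.
; UpdateFX sends energy percentage to the cue's audio components and, after a step-count delay, gradually lowers additive kick gain through the target's AudioParameterComponent.
; Removal resets the fade value and step counter.
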(event EventOnLoopingStart(EitherVFX-preferthisversion)
  (params (MyTarget :type Actor%20Object%20Reference) (Parameters :type Gameplay%20Cue%20Parameters%20Structure%20%28by%20ref%29) (ParticleComponents :type Array%20of%20FXSystem%20Component%20Object%20References) (AudioComponents :type Array%20of%20Audio%20Component%20Object%20References))
  (bind _target (Variables|Default|GetTarget))
  (bind _returnvalue (Actor|GetComponentbyClass _target "/TacticalSprintGame/Gameplay/TacticalSprintEnergyComponent.TacticalSprintEnergyComponent_C"))
  (if (not (CallFunction|IsNativized))
    (Networking|CanExecuteCosmeticEvents
      (:True
        (Variables|Default|SetAudio%20Components AudioComponents)
        (Variables|Default|SetTarget MyTarget)
        (Utilities|Time|ClearandInvalidateTimerbyHandle 0 (Variables|Default|GetStepReset))
        (Utilities|IsLocallyViewed MyTarget
          (:Is%20Locally%20Viewed
            (Utilities|IsValid _returnvalue
              (:Is%20Valid
                (Variables|Default|SetTacSprintComp _returnvalue)
                (Utilities|Time|SetTimerbyEvent (EventDispatchers|CreateEvent :!event Custom|UpdateFX) (Variables|Default|GetFrequencyToCheckEnergy) true) ; Only operates for local players.
              ))))))))

(event Custom|UpdateFX
  (bind _target (Variables|Default|GetTarget))
  (bind _returnvalue (Actor|GetComponentbyClass _target "/Script/AudioGameplay.AudioParameterComponent"))
  (bind _timedfadeval (Variables|Default|GetTimedFadeVal))
  ;[1 1 1] Updates Energy Percentage variable in associated audio component.  This only needs to run for locally controlled or spectated pawns.
  (Variables|Default|SetCurrentEnergyPercentage (Energy|GetCurrentEnergyPercentage (Variables|Default|GetTacSprintComp)))
  (for _array_element (Variables|Default|GetAudio%20Components)
    (Audio|Parameter|SetFloatParameter _array_element "EnergyPercentage" (Variables|Default|GetCurrentEnergyPercentage) :!class /Script/AudioExtensions.AudioParameterControllerInterface)
    ;[1 1 1] Lower additive kicks volume overtime
    (Math|Integer|IncrementInt (Variables|Default|GetCurrentStepCount)
      (:%20%20
        (Math|Integer|CompareInt (Variables|Default|GetCurrentStepCount) (Variables|Default|GetPreStepFadeDelay)
          (:>
            _)
          (:==
            _))
        (Utilities|IsValid _returnvalue
          (:Is%20Valid
            (Audio|Parameter|SetFloatParameter _returnvalue "KickAdditiveGainFade" (Variables|Default|GetTimedFadeVal) :!class /Script/AudioExtensions.AudioParameterControllerInterface)
            (if (>= _timedfadeval (Variables|Default|GetLowestGainLevel))
              (Variables|Default|SetTimedFadeVal (- _timedfadeval 0.01)))))))))

(event EventOnRemoval(EitherVFX-preferthisversion)
  (params (MyTarget :type Actor%20Object%20Reference) (Parameters :type Gameplay%20Cue%20Parameters%20Structure%20%28by%20ref%29) (ParticleComponents :type Array%20of%20FXSystem%20Component%20Object%20References) (AudioComponents :type Array%20of%20Audio%20Component%20Object%20References) (BurstCameraShakeInstance :type Camera%20Shake%20Base%20Object%20Reference) (BurstDecalInstance :type Decal%20Actor%20Object%20Reference))
  (if (not (CallFunction|IsNativized))
    (Networking|CanExecuteCosmeticEvents
      (:True
        (Utilities|IsLocallyViewed MyTarget
          (:Is%20Locally%20Viewed
            (Variables|Default|SetTimedFadeVal 1.0)
            (Variables|Default|SetCurrentStepCount 0)))))))

(fn ConstructionScript)
