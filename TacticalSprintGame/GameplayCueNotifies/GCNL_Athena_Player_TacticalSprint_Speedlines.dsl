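; Looping gameplay cue for tactical sprint speedline effects, feeding the sprinting player's energy into Niagara.
; On loop start it caches the character, sets bIsLocalClient on the first Niagara component and, for locally viewed players only, starts a repeating timer.
; Each tick of that timer writes TacticalSprintEnergyComponent's energy percentage to the "energy" parameter; removal clears the timer and zeroes it.
; A CheckSurfaceType helper maps snow, sand and ice to integers but is not called here.
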(event EventOnLoopingStart(Niagara)
  (params (MyTarget :type Actor%20Object%20Reference) (Parameters :type Gameplay%20Cue%20Parameters%20Structure%20%28by%20ref%29) (ParticleComponents :type Array%20of%20Niagara%20Particle%20System%20Component%20Object%20References) (AudioComponents :type Array%20of%20Audio%20Component%20Object%20References))
  ;[1 1 1] Initialize "bIsLocalClient" variable on VFX and register for updates for Jumping and Energy Percentage
  (bind _ascharacter (Utilities|Casting|CastToCharacter MyTarget))
  (bind _output_get (Variables|Default|SetTargetCharacter _ascharacter))
  (Variables|Default|SetIsLocallyViewed (Pawn|IsLocallyViewed _output_get))
  (bind _asfort_player_pawn (Utilities|Casting|CastToFortPlayerPawn MyTarget))
  (Variables|Default|SetAs%20Fort%20Player%20Pawn _asfort_player_pawn)
  (Variables|Default|SetParticle%20Components ParticleComponents)
  (bind _zeroth_niagara_particle_system_component (Utilities|GetZerothNiagaraParticleSystemComponent)
    (:Is%20Valid
      (Niagara|SetNiagaraVariable(Bool) _zeroth_niagara_particle_system_component "bIsLocalClient" (Variables|Default|GetIsLocallyViewed))
      (if (Variables|Default|GetIsLocallyViewed)
        (bind _returnvalue (Utilities|Time|SetTimerbyEvent (EventDispatchers|CreateEvent :!event Custom|LoopUpdate) (Variables|Default|GetTimeBetweenLoopUpdates) true))
        (Variables|Default|SetLoopUpdateTimer _returnvalue)))))

(event EventOnRemoval(EitherVFX-preferthisversion)
  (params (MyTarget :type Actor%20Object%20Reference) (Parameters :type Gameplay%20Cue%20Parameters%20Structure%20%28by%20ref%29) (ParticleComponents :type Array%20of%20FXSystem%20Component%20Object%20References) (AudioComponents :type Array%20of%20Audio%20Component%20Object%20References) (BurstCameraShakeInstance :type Camera%20Shake%20Base%20Object%20Reference) (BurstDecalInstance :type Decal%20Actor%20Object%20Reference))
  ;[1 1 1] Unbind events
  (Utilities|IsValid (Variables|Default|GetTargetCharacter)
    (:Is%20Valid
      (if (Variables|Default|GetIsLocallyViewed)
        (Utilities|Time|ClearandInvalidateTimerbyHandle 0 (Variables|Default|GetLoopUpdateTimer))
        (bind _zeroth_niagara_particle_system_component (Utilities|GetZerothNiagaraParticleSystemComponent)
          (:Is%20Valid
            (Niagara|SetNiagaraVariable(Float) _zeroth_niagara_particle_system_component "energy" 0.0)))))))

(event Custom|LoopUpdate
  ;[1 1 1] Update Current Energy Percentage in Niagara FX
  (Utilities|IsValid (Variables|Default|GetTargetCharacter)
    (:Is%20Valid
      (bind _zeroth_niagara_particle_system_component (Utilities|GetZerothNiagaraParticleSystemComponent)
        (:Is%20Valid
          (Niagara|SetNiagaraVariable(Float) _zeroth_niagara_particle_system_component "energy" (CallFunction|GetPercentageSprintEnergy (Variables|Default|GetTargetCharacter))))))))

(fn CheckSurfaceType
  (params (Pawn :type Fort%20Pawn%20Object%20Reference))
  (bind _returnvalue (MovementAudio|GetFootstepSurfaceType Pawn))
  (if (Utilities|Enum|Equal(Enum) _returnvalue "Snow")
    (Variables|Default|SetSurfaceTypeInt 1)
    (elif (Utilities|Enum|Equal(Enum) _returnvalue "Sand")
      (Variables|Default|SetSurfaceTypeInt 2)
      (elif (Utilities|Enum|Equal(Enum) _returnvalue "Ice")
        (Variables|Default|SetSurfaceTypeInt 3)
        (else
          (Variables|Default|SetSurfaceTypeInt 0)))))
  (return (Variables|Default|GetSurfaceTypeInt)))

(fn GetPercentageSprintEnergy
  (params (Energy%20Owning%20Actor :type Actor%20Object%20Reference))
  (bind _returnvalue (Actor|GetComponentbyClass Energy%20Owning%20Actor "/TacticalSprintGame/Gameplay/TacticalSprintEnergyComponent.TacticalSprintEnergyComponent_C"))
  (Utilities|IsValid Energy%20Owning%20Actor
    (:Is%20Valid
      (Utilities|IsValid _returnvalue
        (:Is%20Valid
          (return (Energy|GetCurrentEnergyPercentage _returnvalue)))
        (:Is%20Not%20Valid
          _)) ; Does the owning actor have a Tactical Sprint Energy component?
    )
    (:Is%20Not%20Valid
      _)) ; Is Actor that Owns Tactical Sprint Energy Component Valid?\n(It should be on the Fort Player Pawn.)
  (return 0.0))

(fn ConstructionScript)
